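# Translucent UI master material for frontend challenge plates, drawing either a skewable quad or a glowing background.
# Four rotated-UV threshold masks multiply into the quad shape.
# Its fill is a texture or a Primary-to-Secondary gradient, optionally with a power-adjusted inner shape from offset copies of those masks.
# The GlowBackground switch substitutes an aspect-corrected radial gradient as mask, coloring a panning triangle grid and hotspot between Primary and Secondary.
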Material /Game/Athena/UI/Frontend/GrantChallenge/M-UI-Challene-Plate-Master
graph {
 properties {
  bUsedWithStaticMesh=false
  blendMode=BLEND_Translucent
  materialDomain=MD_UI
  pixelDepthOffsetMode=PDOM_Legacy
  usedShadingModels=Default Lit
 }
 nodes [
  LinearInterpolate(
   A=VectorParameter() {parameterName=Accent, defaultValue=<rgba 0 0 1 1>}.RGB
   B=Constant() {r=1}
   Alpha=ScalarParameter() {defaultValue=0.25, group=ColorGradient, parameterName=WhiteFillAmount}
  )
  n1 = ScalarParameter() {defaultValue=0.5, parameterName=Bot-Angle}
  n2 = ComponentMask(None=Rotator(Coordinate=TextureCoordinate(), Time=@n1) {speed=6.2832}) {g=true}
  n3 = ScalarParameter() {parameterName=Bot-Subtract}
  n4 = ScalarParameter() {defaultValue=512, parameterName=Bot-Edge}
  n5 = Clamp(None=Multiply(A=Subtract(A=@n2, B=@n3), B=@n4))
  n6 = ScalarParameter() {defaultValue=-0.01, parameterName=Top-Angle}
  n7 = ComponentMask(None=Rotator(Coordinate=TextureCoordinate(), Time=@n6) {speed=6.2832}) {g=true}
  n8 = ScalarParameter() {defaultValue=0.05, parameterName=Top-Subtract}
  n9 = ScalarParameter() {defaultValue=512, parameterName=Top-Edge}
  n10 = Clamp(None=Multiply(A=Subtract(A=@n7, B=@n8), B=@n9))
  n11 = ScalarParameter() {defaultValue=0.5, parameterName=Right-Angle}
  n12 = ComponentMask(None=Rotator(Coordinate=TextureCoordinate(), Time=@n11) {speed=6.2832}) {r=true}
  n13 = ScalarParameter() {parameterName=Right-Subtract}
  n14 = ScalarParameter() {defaultValue=512, parameterName=Right-Edge}
  n15 = Clamp(None=Multiply(A=Subtract(A=@n12, B=@n13), B=@n14))
  n16 = ScalarParameter() {parameterName=Left-Angle}
  n17 = ComponentMask(None=Rotator(Coordinate=TextureCoordinate(), Time=@n16) {speed=6.2832}) {r=true}
  n18 = ScalarParameter() {parameterName=Left-Subtract}
  n19 = ScalarParameter() {defaultValue=512, parameterName=Left-Edge}
  n20 = Clamp(None=Multiply(A=Subtract(A=@n17, B=@n18), B=@n19))
  n21 = MaterialFunctionCall() {materialFunction=&GetUserInterfaceUV}
  n22 = MaterialFunctionCall(Float2=@n21.Pixel Size) {materialFunction=&BreakOutFloat2Components}
  n23 = AppendVector(A=Divide(A=@n22.R, B=Max(A=@n22.G)), B=Constant() {r=1})
  n24 = MaterialFunctionCall(
   UVs=Multiply(A=@n23, B=@n21.Normalized UV)
   CenterPosition=Multiply(A=@n23) {constB=0.5}
   Radius=ScalarParameter() {defaultValue=1, parameterName=Radius}
   Density=ScalarParameter() {defaultValue=1.5, parameterName=Density}
  ) {materialFunction=&RadialGradientExponential}
  StaticSwitchParameter(True=@n24.RadialGradientExponential, False=DotProduct(A=DotProduct(A=@n5, B=@n10), B=DotProduct(A=@n15, B=@n20))) {parameterName=GlowBackground}
  n25 = TextureCoordinate()
  n26 = Multiply(
   A=Rotator(Coordinate=Multiply(A=@n25, B=@n23), Time=ScalarParameter() {defaultValue=0.125, parameterName=TriangleRotation}) {speed=6.2832}
   B=ScalarParameter() {defaultValue=1.5, parameterName=TriangleTile}
  )
  n27 = Multiply(A=Time(), B=ScalarParameter() {defaultValue=0.05, group=TriangleBackground, parameterName=TimeMultiplier})
  n28 = LinearInterpolate(
   B=TextureSample(UVs=Panner(Coordinate=@n26, Time=@n27) {speedX=-0.25, speedY=-0.125}) {texture=/Game/Global/Textures/Default/Masks_and_Gradients/T-TriangleGrid}.RGB
   Alpha=Power(Base=@n24.RadialGradientExponential, Exp=ScalarParameter() {defaultValue=1.5, parameterName=TrianglePower})
  )
  n29 = ScalarParameter() {defaultValue=0.5, parameterName=HotspotSize}
  n30 = Saturate(None=Multiply(A=Subtract(A=@n24.RadialGradientExponential, B=@n29), B=ScalarParameter() {defaultValue=3, parameterName=HotsportEdge}))
  n31 = VectorParameter() {parameterName=Secondary, defaultValue=<rgba 0 1 0 1>}
  n32 = VectorParameter() {parameterName=Primary, defaultValue=<rgba 1 0 0 1>}
  n33 = Multiply(A=TextureCoordinate()) {constB=0.6667}
  n34 = LinearInterpolate(A=@n32.RGB, B=@n31.RGB, Alpha=Saturate(None=ComponentMask(None=@n33) {g=true}))
  n35 = ScalarParameter() {defaultValue=0.0035, parameterName=RightAngle-Offset}
  n36 = ComponentMask(None=Rotator(Coordinate=TextureCoordinate(), Time=Add(A=@n11, B=@n35)) {speed=6.2832}) {r=true}
  n37 = Add(A=@n13, B=ScalarParameter() {defaultValue=0.035, parameterName=EdgeOffSet})
  n38 = Clamp(None=Multiply(A=Subtract(A=@n36, B=@n37), B=@n14))
  n39 = ScalarParameter() {defaultValue=-0.0035, parameterName=LeftAngle-Offset}
  n40 = ComponentMask(None=Rotator(Coordinate=TextureCoordinate(), Time=Add(A=@n16, B=@n39)) {speed=6.2832}) {r=true}
  n41 = Add(A=@n18, B=ScalarParameter() {defaultValue=0.035, parameterName=EdgeOffSet})
  n42 = Clamp(None=Multiply(A=Subtract(A=@n40, B=@n41), B=@n19))
  n43 = ScalarParameter() {defaultValue=0.0035, parameterName=BotAngle-Offset}
  n44 = ComponentMask(None=Rotator(Coordinate=TextureCoordinate(), Time=Add(A=@n1, B=@n43)) {speed=6.2832}) {g=true}
  n45 = Add(A=@n3, B=ScalarParameter() {defaultValue=0.035, parameterName=EdgeOffSet})
  n46 = Clamp(None=Multiply(A=Subtract(A=@n44, B=@n45), B=@n4))
  n47 = ScalarParameter() {defaultValue=0.0035, parameterName=TopAngle-Offset}
  n48 = ComponentMask(None=Rotator(Coordinate=TextureCoordinate(), Time=Add(A=@n6, B=@n47)) {speed=6.2832}) {g=true}
  n49 = Add(A=@n8, B=ScalarParameter() {defaultValue=0.035, parameterName=EdgeOffSet})
  n50 = Clamp(None=Multiply(A=Subtract(A=@n48, B=@n49), B=@n9))
  n51 = LinearInterpolate(
   A=@n34
   B=Power(Base=@n34, Exp=ScalarParameter() {parameterName=Power})
   Alpha=DotProduct(A=DotProduct(A=@n38, B=@n42), B=DotProduct(A=@n46, B=@n50))
  )
  StaticSwitchParameter(
   True=LinearInterpolate(A=@n32.RGB, B=@n31.RGB, Alpha=Add(A=@n28, B=@n30))
   False=StaticSwitchParameter(
    True=TextureSampleParameter2D() {parameterName=Texture, texture=/Game/Athena/UI/Challenges/Art/TileImages/T_UI_Challenge_AlterEgo_Custom}.RGB
    False=StaticSwitchParameter(True=@n51, False=@n34) {defaultValue=true, group=InnerShape, parameterName=InnerShape}
   ) {group=InnerShape, parameterName=Texture}
  ) {parameterName=GlowBackground}
 ]
}
